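# Material function for the Victory Highway gadget prop that adds a three-color heat glow to incoming material attributes.
# A smoothed gradient of pre-skinned local Z over the bounds minimum, scaled by an inverted squared rerouted value, blends low, mid and high temperature colors.
# The result feeds squared, multiplier-scaled boosts to BaseColor and EmissiveColor.
# It also declares a timestamp transition with a static-switch debug time loop.
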MaterialFunction /BRCosmetics/Animation/Props/Gadgets_VictoryHighway/Materials/MF_VictoryHighway_TechArt
graph {nodes [
  n1 = GetMaterialAttributes(None=NamedRerouteUsage()) {attributeGetTypes [#1, #2]}
  n2 = LocalPosition() {localOrigin=InstancePreSkinning}
  n3 = ComponentMask(None=PreSkinnedLocalBounds().Min) {b=true}
  n4 = SmoothStep(
   Min=ScalarParameter() {defaultValue=0.2, group=VictoryHighwayTechArt, parameterName=LimitOfHeatGradient}
   Value=VertexInterpolator(VS=Divide(A=@n2.Z, B=@n3)).PS
  ) {constMin=0.728}
  n5 = NamedRerouteUsage()
  n6 = OneMinus(None=Multiply(A=@n5, B=@n5))
  n7 = MaterialFunctionCall(A=NamedRerouteUsage(), B=NamedRerouteUsage(), C=NamedRerouteUsage(), Alpha=Saturate(None=Multiply(A=@n4, B=@n6))) {materialFunction=&Lerp_3Color}
  n8 = Multiply(
   A=Add(A=@n1.EmissiveColor, B=@n7.Result) {constA=1}
   B=ScalarParameter() {defaultValue=1, group=VictoryHighwayTechArt, parameterName=HeatMultiplierEmissiveColor}
  )
  n9 = Multiply(A=@n7.Result, B=ScalarParameter() {defaultValue=1, group=VictoryHighwayTechArt, parameterName=HeatMultiplierBaseColor})
  n10 = Add(B=Multiply(A=@n9, B=@n9)) {constA=1}
  FunctionOutput(None=SetMaterialAttributes!(MaterialAttributes=@n1.MaterialAttributes, EmissiveColor=Multiply(A=@n8, B=@n8), BaseColor=Multiply(A=@n1.BaseColor, B=@n10)) {attributeSetTypes [#1, #2]})
  MaterialFunctionCall!(
   Time
   Divisor=ScalarParameter() {defaultValue=1.5, group=FNCSShield_Screen, parameterName=DebugTimeDivisor, sortPriority=30}
   Mod=ScalarParameter() {defaultValue=2, group=FNCSShield_Screen, parameterName=DebugTimeModulo, sortPriority=30}
  ) {materialFunction=/Game/Effects/Fort_Effects/Materials/_Functions/MF_ModuloTime}
  NamedRerouteDeclaration(None=FunctionInput() {inputName=In_MaterialAttribute, inputType=FunctionInput_MaterialAttributes}) {name=In_MaterialAttribute, nodeColor=<rgba 1 0.3516 0 1>}
  NamedRerouteDeclaration(None=FunctionInput() {inputName=In_TechArtMask, sortPriority=1}) {name=In_TechArtMask, nodeColor=<rgba 0.0312 0 1 1>}
  n11 = Divide(A=Time(), B=ScalarParameter() {defaultValue=3, group=FNCSShield_Screen, parameterName=DebugTransitionLength, sortPriority=30})
  n12 = MaterialFunctionCall(
   TimestampA=ScalarParameter() {defaultValue=1, parameterName=Timestamp, sortPriority=30}
   TransitionLength_In=ScalarParameter() {defaultValue=0.1, parameterName=TransitionLength, sortPriority=30}
  ) {materialFunction=&MF_Timestamp_InOut_01}
  NamedRerouteDeclaration(None=StaticSwitchParameter(True=Frac(None=@n11), False=VertexInterpolator(VS=@n12.Linear).PS) {group=FNCSShield_Screen_Graphics, parameterName=DebugModeOn, sortPriority=0}) {name=TimeStamp, nodeColor=<rgba 0.125 1 0 1>}
  NamedRerouteDeclaration(None=VectorParameter() {group=VictoryHighwayTechArt, parameterName=LowTempColor}.RGB) {name=LowTempColor, nodeColor=<rgba 0.2891 1 0 1>}
  n13 = VectorParameter() {group=VictoryHighwayTechArt, parameterName=MidTempColor, sortPriority=33, defaultValue=<rgba 0.5826 0.1168 0.0912 0>}
  NamedRerouteDeclaration(None=@n13.RGB) {name=MidTempColor, nodeColor=<rgba 1 0.9469 0 1>}
  n14 = VectorParameter() {group=VictoryHighwayTechArt, parameterName=HighTempColor, sortPriority=34, defaultValue=<rgba 2 1.1084 0.4696 0>}
  NamedRerouteDeclaration(None=@n14.RGB) {name=HighTempColor, nodeColor=<rgba 1 0.0551 0.0384 1>}
 ]}
